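; BP_DoofyActorsVisibilityHandler shows or hides a configured list of prop actors, following a broadcast visibility state.
; Once playlist data is ready, it uses the global event router, waiting for it if needed, to listen to the stateful FigmentActorsVisibilityStructure event.
; The payload's visible flag sets hidden-in-game on each valid PropVisibility actor.
; During replays, it also binds scrub completion to repeat the listen-and-apply logic.
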(event EventBeginPlay
  (Class|FortAsyncActionCurrentPlaylistReady|PlaylistDataReadyAsync
    (:OnReady
      (if (Replay|GetIsPlayingReplay)
        (ReplayContext|BindEventtoOnReplayScrubComplete (BlueprintContext|GetFortReplayContext) (delegate Custom|OnScrub))
        (else
          (if (not (Utilities|IsValid (GameplayEventRouter|GetGlobalEventRouter) :!class /Script/Engine.KismetSystemLibrary))
            (GameplayEventRouter|WaitForGlobalEventRouter 0
              (:OnEventRouterAvailable
                _)))))
      (bind (_asynctaskproxy _payload) (GameplayEventRouter|StartListeningToStatefulEvent (GameplayEventRouter|GetGlobalEventRouter) "/Figment_S08/Athena/Prototype/Blueprints/White/FigmentActorsVisibilityStructure.FigmentActorsVisibilityStructure")
        (:OnEventReceived
          _)
        (:OnSavedEventStateExists
          _))
      (for _array_element (Variables|Default|GetPropVisibility)
        (Utilities|IsValid _array_element
          (:Is%20Valid
            (bind _payload_bvisible_5_fc7b282248d7a270061af995878e83eb (Utilities|Struct|BreakFigmentActorsVisibilityStructure _payload))
            (Rendering|SetActorHiddenInGame _array_element (not _payload_bvisible_5_fc7b282248d7a270061af995878e83eb))))))))

(event Custom|OnScrub
  (if (not (Utilities|IsValid (GameplayEventRouter|GetGlobalEventRouter) :!class /Script/Engine.KismetSystemLibrary))
    (GameplayEventRouter|WaitForGlobalEventRouter 0
      (:OnEventRouterAvailable
        _)))
  (bind (_asynctaskproxy _payload) (GameplayEventRouter|StartListeningToStatefulEvent (GameplayEventRouter|GetGlobalEventRouter) "/Figment_S08/Athena/Prototype/Blueprints/White/FigmentActorsVisibilityStructure.FigmentActorsVisibilityStructure")
    (:OnEventReceived
      _)
    (:OnSavedEventStateExists
      _))
  (for _array_element (Variables|Default|GetPropVisibility)
    (Utilities|IsValid _array_element
      (:Is%20Valid
        (bind _payload_bvisible_5_fc7b282248d7a270061af995878e83eb (Utilities|Struct|BreakFigmentActorsVisibilityStructure _payload))
        (Rendering|SetActorHiddenInGame _array_element (not _payload_bvisible_5_fc7b282248d7a270061af995878e83eb))))))

(fn ConstructionScript)
