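; BP_DoofyHelper is a small helper Blueprint that triggers a snow-ending cinematic.
; Its custom event PlaySnowEnd gets the sequence player from the level sequence actor in the SnowEndSequence variable and plays it.
; It does no validity check on that reference, and the construction script is empty.
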(event Custom|PlaySnowEnd
  (Sequencer|Player|Play (Variables|/Script/LevelSequence.LevelSequenceActor|GetSequencePlayer (Variables|Default|GetSnowEndSequence)) :!class /Script/MovieScene.MovieSceneSequencePlayer))

(fn ConstructionScript)
