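; Always-relevant actor that streams in the event sublevels LeakyMap and SleepyMap.
; On BeginPlay it waits for mesh network readiness and acts only on Edge nodes, loading both levels immediately if AlwaysLoadLevels is set.
; Otherwise it waits for playlist data, skips Creative, and uses calendar events GW1 (Mini Kev) and GW2 (Beacons) to set LoadSleepyMap and LoadLeakyMap.
; Those variables' OnRep functions stream in the matching level.
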(event EventBeginPlay
  (bind _nodetype (Class|FortAsyncActionMeshNetworkReady|MeshNetworkReadyAsync)
    (:OnReady
      (switch Utilities|FlowControl|Switch|SwitchonEMeshNetworkNodeType _nodetype
        (:Root
          _)
        (:Inner
          _)
        (:Edge
          (if (Variables|Default|GetAlwaysLoadLevels)
            (CallFunction|LoadLeaky)
            (CallFunction|LoadSleepy)
            (else
              (bind _gamestate (Class|FortAsyncActionCurrentPlaylistReady|PlaylistDataReadyAsync)
                (:OnReady
                  (if (not (GameplayTags|HasTag (Class|FortGameStateAthena|GetAthenaPlaylistContextTags _gamestate) "(TagName=\"Athena.Creative\")"))
                    (Utilities|FlowControl|Sequence
                      (:then_0
                        (bind _biseventactive (CalendarEvents|GetCalendarEventInformation "GW1")) ; Mini kev challenge
                        (if _biseventactive
                          (Variables|Default|SetLoadSleepyMap true)))
                      (:then_1
                        (bind _biseventactive_2 (CalendarEvents|GetCalendarEventInformation "GW2")) ; Beacons challenge
                        (if _biseventactive_2
                          (Variables|Default|SetLoadLeakyMap true))))))) ; This only runs on game servers and clients
            )))
        (:Client
          _)))))

(event Custom|LoadLeaky
  (Game|LoadStreamLevel(byName) "LeakyMap" true))

(event Custom|LoadSleepy
  (Game|LoadStreamLevel(byName) "SleepyMap" true))

(event Collision|EventActorBeginOverlap
  (params (OtherActor :type Actor%20Object%20Reference)))

(event EventTick
  (params (DeltaSeconds :type Float%20%28single-precision%29)))

(fn OnRep_LoadLeakyMap
  (CallFunction|LoadLeaky))

(fn OnRep_LoadSleepyMap
  (CallFunction|LoadSleepy))

(fn ConstructionScript)
