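; Snow countdown Blueprint that displays a hologram countdown with a skybeam, apparently for an event.
; The construction script raises the countdown to HologramHeight and centers the skybeam at half that height, scaling its Z to match.
; Debug and Debug_Phase3 booleans set the show-countdown and phase scalars in the shared MPC_Countdown material parameter collection, clearing them when Debug is off.
; BeginPlay, Tick and overlap events are empty.
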(event EventBeginPlay)

(event Collision|EventActorBeginOverlap
  (params (OtherActor :type Actor%20Object%20Reference)))

(event EventTick
  (params (DeltaSeconds :type Float%20%28single-precision%29)))

(fn ConstructionScript
  (bind _hologramheight (Variables|Default|GetHologramHeight))
  (bind _returnvalue (Math|Vector|MakeVector 0.0 0.0 _hologramheight))
  (bind _hologramskybeam (Variables|Default|GetHologramSkybeam))
  (bind (_returnvalue_x _returnvalue_y) (Math|Vector|BreakVector (Transformation|GetWorldScale _hologramskybeam)))
  (Utilities|FlowControl|Sequence
    (:then_1
      (Transformation|SetRelativeLocation (Variables|Default|GetHologramCountdown) _returnvalue false false)
      (Transformation|SetRelativeLocation _hologramskybeam (/ _returnvalue 2.0) false false)
      (Transformation|SetRelativeScale3D _hologramskybeam (Math|Vector|MakeVector _returnvalue_x _returnvalue_y (/ (.z _returnvalue) 1.2e+02))))
    (:then_2
      (if (Variables|Default|GetDebug)
        (Rendering|Material|SetScalarParameterValue "/Game/Athena/Events/Generic/CountdownTimer/Materials/MPC_Countdown.MPC_Countdown" "ShowCountdown" 1.0 :!class /Script/Engine.KismetMaterialLibrary)
        (Rendering|Material|SetScalarParameterValue "/Game/Athena/Events/Generic/CountdownTimer/Materials/MPC_Countdown.MPC_Countdown" "ShowCountdownPhase2" 1.0 :!class /Script/Engine.KismetMaterialLibrary)
        (if (Variables|Default|GetDebug_Phase3)
          (Rendering|Material|SetScalarParameterValue "/Game/Athena/Events/Generic/CountdownTimer/Materials/MPC_Countdown.MPC_Countdown" "ShowCountdownPhase3" 1.0 :!class /Script/Engine.KismetMaterialLibrary)
          (else
            (Rendering|Material|SetScalarParameterValue "/Game/Athena/Events/Generic/CountdownTimer/Materials/MPC_Countdown.MPC_Countdown" "ShowCountdownPhase3" :!class /Script/Engine.KismetMaterialLibrary)))
        (else
          (Rendering|Material|SetScalarParameterValue "/Game/Athena/Events/Generic/CountdownTimer/Materials/MPC_Countdown.MPC_Countdown" "ShowCountdown" :!class /Script/Engine.KismetMaterialLibrary)
          (Rendering|Material|SetScalarParameterValue "/Game/Athena/Events/Generic/CountdownTimer/Materials/MPC_Countdown.MPC_Countdown" "ShowCountdownPhase2" :!class /Script/Engine.KismetMaterialLibrary)
          (Rendering|Material|SetScalarParameterValue "/Game/Athena/Events/Generic/CountdownTimer/Materials/MPC_Countdown.MPC_Countdown" "ShowCountdownPhase3" :!class /Script/Engine.KismetMaterialLibrary))))))
