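; Staged buildup effect actor that shows wispy smoke and embers, hidden at stage zero.
; SetEffectsState stores the stage, swaps the smoke particle template and sets mesh scale, opacity and glow, instantly when not at runtime.
; Runtime changes wait for the gateway pedestal animation, fire a local quake cue, then drive the mesh through a timeline.
; The construction script applies the saved stage; SetEffectsState2 is a hard-coded variant, uncalled here.
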(event Custom|SetEffectsState
  (params (Stage :type Integer) (Runtime :type Boolean))
  (bind _buildupstage (Variables|Default|GetBuildupStage))
  (bind _returnvalue (>= _buildupstage 1))
  (bind _p_snow_doofy_wispy_smoke (Variables|Default|GetP_Snow_Doofy_Wispy_Smoke))
  (bind _sm_snow_doofy_wispysmoke (Variables|Default|GetSM_Snow_Doofy_WispySmoke))
  (bind _buildupstage_2 (Variables|Default|GetBuildupStage))
  (bind _sm_snow_doofy_wispysmoke_2 (Variables|Default|GetSM_Snow_Doofy_WispySmoke))
  (if Runtime
    (if (== Stage (Variables|Default|GetBuildupStage))
      _
      (else
        _)))
  (Variables|Default|SetBuildupStage Stage)
  (Rendering|SetActorHiddenInGame (<= Stage 0))
  (Effects|Components|ParticleSystem|SetTemplate _p_snow_doofy_wispy_smoke (select (>= _buildupstage 2) "/Figment_S08/Athena/Prototype/Blueprints/White/FX/P_Snow_Doofy_Wispy_Smoke_2.P_Snow_Doofy_Wispy_Smoke_2" "/Figment_S08/Athena/Prototype/Blueprints/White/FX/P_Snow_Doofy_Wispy_Smoke_1.P_Snow_Doofy_Wispy_Smoke_1"))
  (Components|Activation|SetActive _p_snow_doofy_wispy_smoke _returnvalue false)
  (if (and _returnvalue Runtime)
    ;[1 1 1] Update during gameplay
    (Rendering|Material|SetScalarParameterValueonMaterials (Variables|Default|GetSM_Snow_Doofy_WispySmoke) "Opacity" (Utilities|Select :Index (Variables|Default|GetBuildupStage) :Option%200 0.0 :Option%201 0.0 :Option%202 (Variables|Default|GetStage1Opacity)))
    (Utilities|FlowControl|Delay 14.0) ; Give the gateway pedestal animation time to finish
    (GameplayCue|ExecuteGameplayCue%28Burst%29Local (Pawn|GetControlledPawn (Utilities|Array|Get(acopy) (Fortnite|GetLocalFortPlayerControllers true true))) (Variables|Default|GetCueTag_BuildupQuake) (Ability|GameplayCue|MakeGameplayCueParameters :NormalizedMagnitude 0.0 :RawMagnitude 0.0 :Location "0,0,0" :Normal "0,0,0" :EffectCauser self :GameplayEffectLevel 1 :AbilityLevel 1 :bReplicateLocationWhenUsingMinimalRepProxy false)) ; Gameplay cue with audio and cam shake
    (bind (_direction _lerp) (Timeline|StateTransition PlayFromStart)
      (:Update
        (Utilities|FlowControl|Sequence
          (:then_0
            (Transformation|SetWorldScale3D (Variables|Default|GetSM_Snow_Doofy_WispySmoke) (Math|Vector|SelectVector (Variables|Default|GetStage2Scale) (Math|Vector|Lerp(Vector) (Variables|Default|GetStage1Scale) (Variables|Default|GetStage2Scale) _lerp) (== (Variables|Default|GetBuildupStage) 2))))
          (:then_1
            (Rendering|Material|SetScalarParameterValueonMaterials _sm_snow_doofy_wispysmoke "Opacity" (Math|Float|Lerp (Utilities|Select :Index _buildupstage_2 :Option%200 0.0 :Option%201 0.0 :Option%202 (Variables|Default|GetStage1Opacity)) (Utilities|Select :Index _buildupstage_2 :Option%200 0.0 :Option%201 (Variables|Default|GetStage1Opacity) :Option%202 (Variables|Default|GetStage2OpacityGlow)) _lerp))
            (Rendering|Material|SetScalarParameterValueonMaterials _sm_snow_doofy_wispysmoke "BaseGlow" (Math|Float|Lerp (Utilities|Select :Index _buildupstage_2 :Option%200 0.0 :Option%201 0.0 :Option%202 (Variables|Default|GetStage1Glow)) (Utilities|Select :Index _buildupstage_2 :Option%200 0.0 :Option%201 (Variables|Default|GetStage1Glow) :Option%202 (Variables|Default|GetStage2OpacityGlow)) _lerp))))))
    (else
      (switch int Stage
        (:1
          (Transformation|SetWorldScale3D _sm_snow_doofy_wispysmoke_2 (Variables|Default|GetStage1Scale))
          (Rendering|Material|SetScalarParameterValueonMaterials _sm_snow_doofy_wispysmoke_2 "Opacity" (Variables|Default|GetStage1Opacity))
          (Rendering|Material|SetScalarParameterValueonMaterials _sm_snow_doofy_wispysmoke_2 "BaseGlow" (Variables|Default|GetStage1Glow))
          (Rendering|SetVisibility (Variables|Default|GetEmbersTop) false false))
        (:2
          (Transformation|SetWorldScale3D _sm_snow_doofy_wispysmoke_2 (Variables|Default|GetStage2Scale))
          (Rendering|Material|SetScalarParameterValueonMaterials _sm_snow_doofy_wispysmoke_2 "Opacity" 1.0)
          (Rendering|Material|SetScalarParameterValueonMaterials _sm_snow_doofy_wispysmoke_2 "BaseGlow" 1.0)
          (Rendering|SetVisibility (Variables|Default|GetEmbers) true false))))))

(fn SetEffectsState2
  (params (Stage :type Integer))
  (bind _p_snow_doofy_wispy_smoke (Variables|Default|GetP_Snow_Doofy_Wispy_Smoke))
  (bind _sm_snow_doofy_wispysmoke (Variables|Default|GetSM_Snow_Doofy_WispySmoke))
  (bind _embers (Variables|Default|GetEmbers))
  (Rendering|SetActorHiddenInGame (<= Stage 0))
  (switch int Stage
    (:1
      (Effects|Components|ParticleSystem|SetTemplate _p_snow_doofy_wispy_smoke "/Figment_S08/Athena/Prototype/Blueprints/White/FX/P_Snow_Doofy_Wispy_Smoke_1.P_Snow_Doofy_Wispy_Smoke_1")
      (Transformation|SetWorldScale3D _sm_snow_doofy_wispysmoke "5.000000,5.000000,7.000000")
      (Rendering|Material|SetScalarParameterValueonMaterials _sm_snow_doofy_wispysmoke "Opacity" 0.5)
      (Rendering|Material|SetScalarParameterValueonMaterials _sm_snow_doofy_wispysmoke "BaseGlow" 0.2)
      (Transformation|SetWorldScale3D _embers "1.000000,1.000000,1.000000"))
    (:2
      (Effects|Components|ParticleSystem|SetTemplate _p_snow_doofy_wispy_smoke "/Figment_S08/Athena/Prototype/Blueprints/White/FX/P_Snow_Doofy_Wispy_Smoke_2.P_Snow_Doofy_Wispy_Smoke_2")
      (Transformation|SetWorldScale3D _sm_snow_doofy_wispysmoke "9.000000,9.000000,9.000000")
      (Rendering|Material|SetScalarParameterValueonMaterials _sm_snow_doofy_wispysmoke "Opacity" 1.0)
      (Rendering|Material|SetScalarParameterValueonMaterials _sm_snow_doofy_wispysmoke "BaseGlow" 1.0)
      (Transformation|SetWorldScale3D _embers "1.000000,1.000000,2.000000"))))

(fn ConstructionScript
  (CallFunction|SetEffectsState (Variables|Default|GetBuildupStage)))
