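; Snow sequence Blueprint, apparently a smoke pillar effect, with a base component and screen-attached smoke.
; Construction script creates a dynamic material on the screen smoke, feeding it height, top and bottom radii, and blend distances.
; It derives bounds scale from those dimensions relative to the mesh's local bounds, so the bounds track the configured smoke volume.
; Custom events lower and reset the pillar base's translucent sort priority.
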(event Custom|SetSortPriority
  (Rendering|SetTranslucentSortPriority (Variables|Default|GetSmokePillarBase) -1))

(event Custom|ResetSortPriority
  (Rendering|SetTranslucentSortPriority (Variables|Default|GetSmokePillarBase) 0))

(fn ConstructionScript
  (bind (_min _max) (Components|StaticMesh|GetLocalBounds (Variables|Default|GetScreenAttachedSmoke)))
  ;[1 1 1] Set up screen attached smoke
  (bind _returnvalue (Rendering|Material|CreateDynamicMaterialInstance (Variables|Default|GetScreenAttachedSmoke) 0 0 "None" :!class /Script/Engine.PrimitiveComponent))
  (Rendering|Material|SetScalarParameterValue _returnvalue "F_RadiusTop" (Variables|ScreenSmoke|GetRadiusTop) :!class /Script/Engine.MaterialInstanceDynamic)
  (Rendering|Material|SetScalarParameterValue _returnvalue "F_BlendRadius" (Variables|ScreenSmoke|GetBlendDistance) :!class /Script/Engine.MaterialInstanceDynamic)
  (Rendering|Material|SetScalarParameterValue _returnvalue "F_RadiusBottom" (Variables|ScreenSmoke|GetRadiusBottom) :!class /Script/Engine.MaterialInstanceDynamic)
  (Rendering|Material|SetScalarParameterValue _returnvalue "F_Height" (Variables|ScreenSmoke|GetHeight) :!class /Script/Engine.MaterialInstanceDynamic)
  (Rendering|Material|SetScalarParameterValue _returnvalue "F_BlendTop" (Variables|ScreenSmoke|GetBlendDistance_Top) :!class /Script/Engine.MaterialInstanceDynamic)
  (Rendering|Material|SetScalarParameterValue _returnvalue "F_BlendBottom" (Variables|ScreenSmoke|GetBlendDistance_Bottom) :!class /Script/Engine.MaterialInstanceDynamic)
  (Rendering|SetBoundsScale (Variables|Default|GetScreenAttachedSmoke) (/ (Math|Float|Max(Float) (Math|Float|float+float (Variables|ScreenSmoke|GetHeight) (Variables|ScreenSmoke|GetBlendDistance_Top) (+ (Variables|ScreenSmoke|GetRadiusTop) (Variables|ScreenSmoke|GetBlendDistance)) (Variables|ScreenSmoke|GetBlendDistance)) (Math|Float|float+float (Variables|ScreenSmoke|GetHeight) (Variables|ScreenSmoke|GetBlendDistance_Bottom) (+ (Variables|ScreenSmoke|GetRadiusBottom) (Variables|ScreenSmoke|GetBlendDistance)) (Variables|ScreenSmoke|GetBlendDistance))) (Math|Float|Max(Float) (.x _max) (.y _max) (.z _max)))))
