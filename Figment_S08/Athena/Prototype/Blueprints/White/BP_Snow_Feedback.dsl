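; BP_Snow_Feedback is a Blueprint bundling camera, controller, audio and material feedback events for a snow-themed effect.
; Shake events share a function that plays world camera shake and near/far force feedback from an Arrow, scaling member radii by a multiplier.
; Post-pulse events play fixed-radius shakes with post effects directly.
; HoloPulse drives material parameter collection scalars via timeline, while ambient override events fade a 2D volcano loop in and out.
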(event Custom|HoloPulse
  (bind (_direction _pulseup _glow) (Timeline|Timeline_0 PlayFromStart)
    (:Update
      (Rendering|Material|SetScalarParameterValue "/Game/Athena/Environments/Festivus/MotorizedLights/Lighting_MPC.Lighting_MPC" "HoloPulseGameTime" _pulseup :!class /Script/Engine.KismetMaterialLibrary)
      (Rendering|Material|SetScalarParameterValue "/Game/Athena/Environments/Festivus/MotorizedLights/Lighting_MPC.Lighting_MPC" "HoloGlow" _glow :!class /Script/Engine.KismetMaterialLibrary))))

(event Custom|SmallShake
  (CallFunction|CameraandControllerShake 1.0 "/Game/Effects/Fort_Effects/BluePrints/CameraShakes/B_Small_Vertical_Jolt_CameraShake.B_Small_Vertical_Jolt_CameraShake_C" "/FNGameplayCues/Weapons/ForceFeedback/FF_Generic_Quick_Jolt_Near.FF_Generic_Quick_Jolt_Near" "/FNGameplayCues/Weapons/ForceFeedback/FF_Generic_Quick_Jolt_Far.FF_Generic_Quick_Jolt_Far"))

(event Custom|AmbientOverrideStart
  (Variables|Default|GetAmbientOverride
    (:then)
    (:else
      (bind _returnvalue (Audio|SpawnSound2D "/Figment_S08/Athena/Sounds/Ambient/Volcano/Amb_Volcano_FloorIsLava_03_Loop_Snow_InProgress_Cue.Amb_Volcano_FloorIsLava_03_Loop_Snow_InProgress_Cue"))
      (bind _output_get (Variables|Default|SetAmbientOverride _returnvalue))
      (Audio|Components|Audio|FadeIn _output_get 3.0 1.0 0.0 "Linear" :!class /Script/Engine.AudioComponent)
      (Development|PrintString "AMBCANO APPLIED"))))

(event Custom|AmbientOverrideEnd
  (bind _ambientoverride (Variables|Default|GetAmbientOverride)
    (:then
      (Audio|Components|Audio|FadeOut _ambientoverride 3.0 0.0 "Linear" :!class /Script/Engine.AudioComponent)
      (Development|PrintString "AMBCANO CLEARED"))
    (:else)))

(event Custom|X-Shake
  (CallFunction|CameraandControllerShake 1.15 "/Figment_S08/Effects/Fort_Effects/BluePrints/CameraShakes/B_X-Axis_Jolt_CameraShake.B_X-Axis_Jolt_CameraShake_C" "/FNGameplayCues/Weapons/ForceFeedback/FF_Generic_Quick_Jolt_Near.FF_Generic_Quick_Jolt_Near" "/FNGameplayCues/Weapons/ForceFeedback/FF_Generic_Quick_Jolt_Far.FF_Generic_Quick_Jolt_Far" true))

(event Custom|WarmupShake
  (CallFunction|CameraandControllerShake 1e+06 "/Figment_S08/Athena/Prototype/Blueprints/White/CamerShake_SnowWarmup.CamerShake_SnowWarmup_C" "/Game/Weapons/ForceFeedback/FF_Generic_Impact_Near.FF_Generic_Impact_Near" "/Game/Weapons/ForceFeedback/FF_Generic_Impact_Far.FF_Generic_Impact_Far"))

(event Custom|BigShake
  (CallFunction|CameraandControllerShake 1e+06 "/Figment_S08/Athena/Prototype/Blueprints/White/CamerShake_SnowBig.CamerShake_SnowBig_C" "/Game/Weapons/ForceFeedback/FF_Generic_Impact_Near.FF_Generic_Impact_Near" "/Game/Weapons/ForceFeedback/FF_Generic_Impact_Far.FF_Generic_Impact_Far"))

(event Custom|TotesHuge
  (CallFunction|CameraandControllerShake 1e+06 "/Figment_S08/Athena/Prototype/Blueprints/White/CamerShake_SnowTotesHuge.CamerShake_SnowTotesHuge_C" "/Game/Weapons/ForceFeedback/FF_Generic_Impact_Near.FF_Generic_Impact_Near" "/Game/Weapons/ForceFeedback/FF_Generic_Impact_Far.FF_Generic_Impact_Far"))

(event Custom|Post_Pulse_Effect
  (Camera|PlayWorldCameraShake "/Figment_S08/Athena/Effects/Particles/Snow/Post/BP_Cam_Shake_With_Post.BP_Cam_Shake_With_Post_C" (Transformation|GetWorldLocation (Variables|Default|GetArrow)) 2.5e+03 7.5e+03))

(event Custom|Post_Pulse_Effect_Long
  (Camera|PlayWorldCameraShake "/Figment_S08/Athena/Effects/Particles/Snow/Post/BP_Cam_Shake_With_Post_Long.BP_Cam_Shake_With_Post_Long_C" (Transformation|GetWorldLocation (Variables|Default|GetArrow)) 2.5e+03 7.5e+03))

(fn CameraAndControllerShake
  (params (RadiusMult :type Float%20%28double-precision%29) (Shake :type Legacy%20Camera%20Shake%20Class%20Reference) (NearForceFeedback :type Force%20Feedback%20Effect%20Object%20Reference) (FarForceFeedback :type Force%20Feedback%20Effect%20Object%20Reference) (OrientShakeToEpicenter :type Boolean))
  (bind _arrow (Variables|Default|GetArrow))
  (bind _returnvalue (Transformation|GetWorldLocation _arrow))
  (bind _outershakeradius (Variables|Default|GetOuterShakeRadius))
  (bind _returnvalue_2 (* _outershakeradius RadiusMult))
  (bind _innershakeradius (Variables|Default|GetInnerShakeRadius))
  (bind _returnvalue_3 (* _innershakeradius RadiusMult))
  (Camera|PlayWorldCameraShake Shake _returnvalue _returnvalue_3 _returnvalue_2 1.0 OrientShakeToEpicenter)
  (Fortnite|PlayLocalForceFeedbackatLocationMulti NearForceFeedback FarForceFeedback _returnvalue (Math|Float|Clamp(Float) (* _returnvalue_3 (Variables|Default|GetVibrationInnerRadiusMult)) 0.0 _returnvalue_2) _returnvalue_2))

(fn ConstructionScript)
