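; BP_Snow_NoBuild is a prototype actor that clears damageable building actors from the area around it when play begins.
; On the authority, BeginPlay runs one large sphere overlap at the root's location, ignoring itself.
; It casts each hit to BuildingActor and destroys those whose bCanBeDamaged flag is set.
; The overlap, tick and construction hooks are empty, so nothing here blocks building afterwards.
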(event EventBeginPlay
  (Networking|SwitchHasAuthority
    (:Authority
      (bind _outactors (Collision|SphereOverlapActors (Transformation|GetWorldLocation (Variables|Default|GetDefaultSceneRoot)) 3e+03 (Utilities|Array|MakeArray "ObjectTypeQuery1") 0 (Utilities|Array|MakeArray self)))
      (for _array_element _outactors
        (bind _asbuilding_actor (Utilities|Casting|CastToBuildingActor _array_element))
        (if (Variables|/Script/Engine.Actor|GetbCanBeDamaged _asbuilding_actor)
          (Actor|DestroyActor _array_element))))))

(event Collision|EventActorBeginOverlap
  (params (OtherActor :type Actor%20Object%20Reference)))

(event EventTick
  (params (DeltaSeconds :type Float%20%28single-precision%29)))

(fn ConstructionScript)
