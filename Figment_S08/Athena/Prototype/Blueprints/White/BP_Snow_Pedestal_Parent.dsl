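; BP_Snow_Pedestal_Parent is a parent Blueprint for snow pedestals in a Figment Season 8 prototype.
; It declares custom events SetServerPinstate_Parent, which takes an E_SnowStates value, and CompleteSequence_Parent, both with empty bodies.
; BeginPlay, Tick, actor overlap and the construction script are also empty, so the file serves as an interface-like base, presumably for child pedestals to implement.
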(event Custom|SetServerPinstate_Parent
  (params (State :type E_SnowStates%20Enum :default "NewEnumerator5")))

(event EventBeginPlay)

(event Collision|EventActorBeginOverlap
  (params (OtherActor :type Actor%20Object%20Reference)))

(event Custom|CompleteSequence_Parent)

(event EventTick
  (params (DeltaSeconds :type Float%20%28single-precision%29)))

(fn ConstructionScript)
