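# Translucent unlit effect material for a cone of scrolling streaks, tinted orange-red by default and apparently tied to lava flows despite its Snow_Cone name.
# A panning swirl texture distorts the U coordinate of a panned, rotated flare-line texture, while vertex green offsets V.
# The streaks multiply a two-color, hue-shifted gradient and, separately, a UV edge-fade mask scaled by LavaFlowsVis from MPC_Snow.
# A CameraOffset function supplies a parameterized world-position offset.
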Material /Figment_S08/Athena/Prototype/Blueprints/White/FX/M_Snow_Cone2
graph {
 properties {blendMode=BLEND_Translucent, pixelDepthOffsetMode=PDOM_Legacy, shadingModel=MSM_Unlit, usedShadingModels=Unlit}
 nodes [
  MaterialFunctionCall(Offset Amount=ScalarParameter() {defaultValue=2500, parameterName="Offset Amount (S)"}) {materialFunction=/Engine/Functions/Engine_MaterialFunctions02/WorldPositionOffset/CameraOffset}
  n1 = TextureCoordinate() {vTiling=1.5}
  n2 = ComponentMask(None=@n1) {g=true}
  n3 = OneMinus(None=@n2)
  n4 = ComponentMask(None=@n1) {r=true}
  n5 = Multiply(A=@n4, B=OneMinus(None=@n4))
  n6 = TextureSample(UVs=Panner() {speedY=-0.1}) {samplerType=SAMPLERTYPE_LinearColor, texture=/Figment_S08/Effects/Fort_Effects/Effects/Environment/Desert/Dust/Organize/T_SwirlFlow_02}
  n7 = ScalarParameter() {defaultValue=0.05, parameterName=Dist_Mult1}
  n8 = TextureCoordinate()
  n9 = ComponentMask(None=@n8) {r=true}
  n10 = ScalarParameter() {defaultValue=0.25, parameterName=TileX_2}
  n11 = AppendVector(A=Add(A=Multiply(A=@n6.R, B=@n7), B=Multiply(A=@n9, B=@n10)), B=Add(A=ComponentMask(None=@n8) {g=true}, B=VertexColor().G))
  n12 = TextureSample(UVs=MaterialFunctionCall(UVs=Panner(Coordinate=@n11) {speedY=-0.05}, "Rotation Angle (0-1)"=ScalarParameter() {parameterName=s}) {materialFunction=&CustomRotator}.Rotated Values) {texture=/Figment_S08/Effects/Fort_Effects/Textures/Flares/T_Flares_Lines_4X1}
  Multiply(
   A=Multiply(A=Multiply(A=@n3, B=@n5), B=Power(Base=@n12.G) {constExponent=0.35})
   B=CollectionParameter() {collection=/Game/Athena/Prototype/Blueprints/White/MPC_Snow, parameterName=LavaFlowsVis}
  )
  n13 = Multiply(A=ComponentMask(None=TextureCoordinate() {vTiling=1.5}) {g=true}, B=VertexColor().B)
  n14 = ScalarParameter() {defaultValue=0.025, parameterName=HueShift}
  n15 = VectorParameter() {parameterName=Color2, defaultValue=<rgba 150 10 0 0>}
  n16 = VectorParameter() {parameterName=Color, defaultValue=<rgba 10 2 0 0>}
  n17 = Multiply(A=Power(Base=@n2), B=VertexColor().G)
  Multiply(
   A=MaterialFunctionCall(Hue Shift Percentage=Multiply(A=@n13, B=@n14), Texture=LinearInterpolate(A=@n15.RGB, B=@n16.RGB, Alpha=@n17)) {materialFunction=&HueShift}.Result
   B=@n12.G
  )
 ]
}
