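# Additive, unlit, two-sided mesh-particle material for the glow of a snow pedestal hole.
# Dynamic parameter Open scales a vertex-color-weighted expansion from the particle position and sets the radius of a UV mask that reveals the glow.
# Two counter-rotating noise samples and vertex red index a rainbow gradient, tinted by particle color and fed to HDR.
# Vertex fades, depth fade and near-camera fade shape opacity; mobile skips expansion and both scene fades.
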Material /Figment_S08/Athena/Prototype/Blueprints/White/FX/M_Snow_Pedestal_Hole_Glow
graph {
 properties {
  bUsedWithMeshParticles=true
  blendMode=BLEND_Additive
  pixelDepthOffsetMode=PDOM_Legacy
  previewMesh=&/SM_Snow_Pedestal_Hole_Glow
  shadingModel=MSM_Unlit
  twoSided=true
  usedShadingModels=Unlit
 }
 nodes [
  ActorPositionWS()
  n1 = VertexColor()
  n2 = Subtract(A=WorldPosition().XYZ, B=ParticlePositionWS())
  n3 = Constant3Vector() {constant=<rgba 0.25 0.25 1 0>}
  n4 = DynamicParameter() {paramNames [Open,Param2,Param3,Param4]}
  n5 = Multiply(
   A=Multiply(A=OneMinus(None=@n1.R), B=Multiply(A=@n2, B=@n3))
   B=Multiply(A=@n4.Open, B=ScalarParameter() {defaultValue=2, parameterName=Expand})
  )
  n6 = Constant()
  FeatureLevelSwitch(Default=@n5, ES2=@n6, ES3_1=@n6, SM4_REMOVED=@n5)
  n7 = ParticleColor()
  n8 = VertexColor()
  n9 = Multiply(A=@n8.R) {constB=1.4}
  n10 = VertexColor()
  n11 = TextureSample(UVs=Rotator(Coordinate=TextureCoordinate() {uTiling=2, vTiling=2}) {centerX=1, centerY=1, speed=0.5}) {texture=/Game/Effects/Fort_Effects/Textures/T_Smooth_tile}
  n12 = TextureSample(UVs=Rotator(Coordinate=TextureCoordinate() {uTiling=4, vTiling=4}) {centerX=2, centerY=2, speed=-0.5}) {texture=&T_Smooth_tile}
  n13 = LinearInterpolate(
   A=ScalarParameter() {parameterName=MinRainbow}
   B=ScalarParameter() {defaultValue=2, parameterName=MaxRainbow}
   Alpha=LinearInterpolate(A=Power(Base=@n10.R) {constExponent=0.6}, B=LinearInterpolate(A=@n11.G, B=@n12.G)) {constAlpha=0.7}
  )
  n14 = Multiply(A=Add(A=Power(Base=@n9), B=TextureSample(UVs=@n13) {texture=/Game/Effects/Fort_Effects/Textures/Grad/T_Gradient_Rainbow_NoRed}.RGB)) {constB=0.5}
  n15 = Multiply(A=@n7.RGB, B=Saturate(None=Add(A=@n8.G, B=@n14)))
  MaterialFunctionCall(LDR Input=Clamp(None=@n15), HDR Tint=@n15, HDR Intensity Multiplier=ScalarParameter() {defaultValue=1, parameterName=HDRIntensity}) {materialFunction=&DeriveHDRfromLDR}
  MaterialFunctionCall() {materialFunction=&DebugTimeSine}
  n16 = Power(Base=@n1.R, Exp=ScalarParameter() {defaultValue=5, parameterName=FadePower}) {constExponent=10}
  n17 = Multiply(A=@n7.A, B=@n16)
  n18 = ComponentMask(None=TextureCoordinate()) {r=true}
  n19 = Constant() {r=0.5}
  n20 = ScalarParameter() {defaultValue=50, parameterName=FadeDistance}
  n21 = Multiply(
   A=MaterialFunctionCall() {materialFunction=&Near_Camera_Fade}.Result
   B=DepthFade(FadeDistance=LinearInterpolate(A=@n20, Alpha=@n16) {constB=0})
  )
  n22 = Constant() {r=1}
  Multiply(
   A=Multiply(A=LinearInterpolate(A=@n17, Alpha=@n1.G) {constA=0.5}, B=SphereMask(A=@n18, B=@n19, Radius=@n4.Open))
   B=FeatureLevelSwitch(Default=@n21, ES2=@n22, ES3_1=@n22, SM4_REMOVED=@n21)
  )
  ObjectRadius()
 ]
}
