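# Translucent, unlit, two-sided mesh-particle material for steam, named for the snow pedestal hole effect.
# Storm-cloud normal and mask textures pan along warped UV1, driving highlights over time-of-day shadow colors and gradient-shaped opacity.
# Inverse vertex red and the Open dynamic parameter scale the mesh offset from particle position; PanAdd shifts the panning.
# View-angle, near-camera and depth fades soften it; mobile drops offset, highlights and the latter two fades.
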Material /Figment_S08/Athena/Prototype/Blueprints/White/FX/M_Snow_Pedestal_Hole_Steam
graph {
 properties {
  bUsedWithMeshParticles=true
  blendMode=BLEND_Translucent
  pixelDepthOffsetMode=PDOM_Legacy
  shadingModel=MSM_Unlit
  twoSided=true
  usedShadingModels=Unlit
 }
 nodes [
  ActorPositionWS()
  n1 = VertexColor()
  n2 = Subtract(A=WorldPosition().XYZ, B=ParticlePositionWS())
  n3 = Constant3Vector() {constant=<rgba 1 1 1 0>}
  n4 = Multiply(
   A=Multiply(A=OneMinus(None=@n1.R), B=Multiply(A=@n2, B=@n3))
   B=Multiply(A=DynamicParameter() {paramNames [Open,PanAdd,Param3,Param4]}.Open, B=ScalarParameter() {defaultValue=2, parameterName=Expand})
  )
  n5 = Constant()
  FeatureLevelSwitch(Default=@n4, ES2=@n5, ES3_1=@n5, SM4_REMOVED=@n4)
  MaterialFunctionCall() {materialFunction=&DebugTimeSine}
  n6 = ParticleColor()
  n7 = ComponentMask(None=TextureCoordinate() {coordinateIndex=1}) {g=true}
  n8 = Constant3Vector() {constant=<rgba 0 1 0 0>}
  n9 = TextureCoordinate() {coordinateIndex=1}
  n10 = Power(Base=OneMinus(None=ComponentMask(None=@n9) {g=true})) {constExponent=3}
  n11 = Constant2Vector() {g=-0.15}
  n12 = Time()
  n13 = ScalarParameter() {defaultValue=1, parameterName=PanSpeed}
  n14 = Panner(
   Coordinate=Add(A=Multiply(A=@n10, B=@n11), B=TextureCoordinate() {coordinateIndex=1, uTiling=6, vTiling=0.25})
   Time=Add(A=Multiply(A=@n12, B=@n13), B=DynamicParameter() {paramNames [Open,PanAdd,Param3,Param4]}.PanAdd)
  ) {speedX=0.1, speedY=1}
  n15 = TextureSample(UVs=@n14) {samplerType=SAMPLERTYPE_Normal, texture=/Game/VisualThreatManager/StormVisuals/LargeClouds/T_StormCloudDetail01_N}
  n16 = Clamp(None=Abs(None=DotProduct(A=@n8, B=@n15.RGB)))
  n17 = ScalarParameter() {defaultValue=3, parameterName=HighlightPower}
  n18 = Add(A=Multiply(A=Power(Base=@n7) {constExponent=5}) {constB=0.5}, B=Multiply(A=@n7, B=Power(Base=@n16, Exp=@n17)))
  n19 = Multiply(
   A=MaterialFunctionCall(
    Color_Morning=VectorParameter() {group=Colors, parameterName=ShadowMorning, defaultValue=<rgba 0.7 0.5855 0.5396 0>}.RGB
    Color_Day=VectorParameter() {group=Colors, parameterName=ShadowDay, defaultValue=<rgba 0.4375 0.496 0.7 0>}.RGB
    Color_Evening=VectorParameter() {group=Colors, parameterName=ShadowEvening, defaultValue=<rgba 0.36 0.3271 0.4 1>}.RGB
    Color_Night=VectorParameter() {group=Colors, parameterName=ShadowNight, defaultValue=<rgba 0.4077 0.4748 0.5 0>}.RGB
   ) {materialFunction=&Color_Time_of_Day}.Result
   B=ScalarParameter() {defaultValue=0.7, parameterName=ShadowMult}
  )
  n20 = QualitySwitch(Default=Add(A=Multiply(A=@n6.RGB, B=@n18) {constB=0}, B=@n19), Low=@n19)
  n21 = Multiply(A=TextureSample(UVs=@n14) {samplerType=SAMPLERTYPE_LinearGrayscale, texture=/Game/VisualThreatManager/StormVisuals/LargeClouds/T_StormCloudDetail01_M}.G)
  Multiply(
   A=Multiply(A=FeatureLevelSwitch(Default=@n20, ES2=@n19, ES3_1=@n19, SM4_REMOVED=@n20), B=LinearInterpolate(Alpha=@n21) {constA=0.4, constB=0.8})
   B=ScalarParameter() {defaultValue=1.5}
  )
  n22 = ComponentMask(None=TextureCoordinate() {coordinateIndex=1}) {g=true}
  n23 = Multiply(
   A=LinearInterpolate(
    A=ScalarParameter() {defaultValue=0.1, parameterName=OpacityMin}
    B=ScalarParameter() {defaultValue=1.5, parameterName=OpacityMax}
    Alpha=Multiply(A=@n21, B=@n22)
   ) {constA=1.4}
   B=Power(Base=@n22) {constExponent=1.5}
  )
  n24 = OneMinus(None=@n1.G)
  n25 = ScalarParameter() {defaultValue=1, parameterName=FadeAngle}
  n26 = ScalarParameter() {defaultValue=70, parameterName=FadeDistance}
  n27 = Multiply(
   A=MaterialFunctionCall(Distance=ScalarParameter() {defaultValue=10, parameterName=NearCamFade}) {materialFunction=&Near_Camera_Fade}.Result
   B=DepthFade(FadeDistance=LinearInterpolate(A=@n26) {constB=5})
  )
  n28 = Constant() {r=1}
  Multiply(
   A=Multiply(A=ParticleColor().A, B=Multiply(A=@n23, B=@n24))
   B=Multiply(
    A=MaterialFunctionCall("FadeSensitivity (optional)"=@n25) {materialFunction=&FadeBasedOnViewAngle}.Result
    B=FeatureLevelSwitch(Default=@n27, ES2=@n28, ES3_1=@n28, SM4_REMOVED=@n27)
   )
  )
 ]
}
