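# Default Lit mesh-particle material, named for an airborne snow rock, built from meteor-chunk diffuse, normal and specular textures.
# A panning smoke normal map on UV1 modulates LavaColor, which fades to black through a vertex-blue mask.
# That map also combines with time-scrolled, actor-relative vector noise, scaled by intensity parameters and blended toward vertex normals by vertex blue, apparently for world position offset.
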Material /Figment_S08/Athena/Prototype/Blueprints/White/FX/M_Snow_Rock_InAir
graph {
 properties {bUsedWithMeshParticles=true, pixelDepthOffsetMode=PDOM_Legacy, usedShadingModels=Default Lit}
 nodes [
  n1 = VectorParameter() {parameterName=LavaColor, defaultValue=<rgba 15 1.25 0 1>}
  n2 = TextureCoordinate() {coordinateIndex=1}
  n3 = AppendVector(A=ScalarParameter() {defaultValue=1, parameterName=TileX2}, B=ScalarParameter() {defaultValue=1, parameterName=TileY2})
  n4 = Time()
  n5 = ScalarParameter() {defaultValue=1, parameterName=PanSpeed}
  n6 = TextureSample(UVs=Panner(Coordinate=Multiply(A=@n2, B=@n3), Time=Multiply(A=@n4, B=@n5)) {speedX=-0.04, speedY=-1}) {samplerType=SAMPLERTYPE_Normal, texture=/Game/Effects/Fort_Effects/Textures/Smoke/T_SmokePuffyTrail2_N}
  n7 = VertexColor()
  n8 = Multiply(A=@n7.B, B=@n7.B) {constA=1}
  LinearInterpolate(A=Add(A=@n1.RGB, B=Multiply(A=@n1.RGB, B=@n6.G)), B=Constant(), Alpha=Clamp(None=Power(Base=@n8) {constExponent=0.1}))
  n9 = ScalarParameter() {defaultValue=250, parameterName=WPO_Intensity}
  n10 = Multiply(A=VertexColor().B, B=ScalarParameter() {defaultValue=500, parameterName=CurlIntensity})
  n11 = Subtract(A=WorldPosition().XYZ, B=ActorPositionWS())
  n12 = LinearInterpolate(Alpha=VertexColor().B)
  n13 = Multiply(A=Multiply(A=ScalarParameter() {defaultValue=-2500, parameterName=Speed}, B=@n12), B=Time())
  n14 = ScalarParameter() {defaultValue=3500, parameterName=Scale}
  n15 = VectorNoise(World Position=Divide(A=Subtract(A=@n11, B=@n13), B=Multiply(A=@n14, B=@n12))) {noiseFunction=VNF_VectorALU}
  LinearInterpolate(A=Add(A=Multiply(A=@n6.RGB, B=@n9), B=Multiply(A=@n10, B=@n15)), B=VertexNormalWS(), Alpha=VertexColor().B)
  TextureSample() {texture=/Game/Environments/Props/Rocks/Textures/T_Meteor_Chunk2_D}
  TextureSample() {samplerType=SAMPLERTYPE_Normal, texture=/Game/Environments/Props/Rocks/Textures/T_Meteor_Chunk2_N}
  TextureSample() {texture=/Game/Environments/Props/Rocks/Textures/T_Meteor_Chunk2_S}
 ]
}
