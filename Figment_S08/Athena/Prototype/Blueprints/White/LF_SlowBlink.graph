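# LF_SlowBlink is a light function material, built to make lights pulse in a slow, repeating blink.
# Time drives the Sine_Remapped material function as its phase, with the period overridden to a one-second cycle.
# The remapped sine is multiplied by itself, squaring the wave to reshape the pulse, though the graph shows no link from that result to a material output.
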Material /Figment_S08/Athena/Prototype/Blueprints/White/LF_SlowBlink
graph {
 properties {materialDomain=MD_LightFunction, pixelDepthOffsetMode=PDOM_Legacy, usedShadingModels=Default Lit}
 nodes [
  n1 = MaterialFunctionCall!(Sine Phase=Time() {bOverride_Period=true, period=1}, Value 1, Value 2) {materialFunction=/Engine/Functions/Engine_MaterialFunctions02/Utility/Sine_Remapped}
  Multiply(A=@n1.Remapped Sine, B=@n1.Remapped Sine)
 ]
}
