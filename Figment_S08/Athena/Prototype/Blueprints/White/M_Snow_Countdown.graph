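# Unlit material for an in-world event countdown: MF_CountdownTimer digits, optionally on a camera-facing sprite.
# Digits take a sine-flickering color plus chromatic smoke or a panning radial rainbow, blending toward an animated barrier-grid fill with MPC_Countdown's phase 3 value.
# Other MPC_Countdown values gate visibility through World Position Offset, which throws hidden geometry far away; Debug switches override both.
# Alpha combines the digit/grid blend with a screen-pixel dot pattern and Erode_Alpha.
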Material /Figment_S08/Athena/Prototype/Blueprints/White/M_Snow_Countdown
graph {
 properties {bEnableResponsiveAA=true, pixelDepthOffsetMode=PDOM_Legacy, shadingModel=MSM_Unlit, usedShadingModels=Unlit}
 nodes [
  n1 = AppendVector(A=ScalarParameter() {defaultValue=220.6335, parameterName=ScaleX}, B=ScalarParameter() {defaultValue=197.6212, parameterName=ScaleY})
  n2 = MaterialFunctionCall() {materialFunction=/Engine/Functions/Engine_MaterialFunctions02/WorldPositionOffset/ObjectScale}
  n3 = MaterialFunctionCall!(Center Location, 0-1 UVs, XY scale=Multiply(A=@n1, B=@n2.Scale Z), OverridePivot, "Normals (Optional)") {materialFunction=&Sprite}
  n4 = Subtract(A=CameraPositionWS(), B=ObjectPositionWS())
  n5 = ScalarParameter() {defaultValue=10117.7236}
  Add(A=@n3.World position offset, B=Multiply(A=Normalize(VectorInput=@n4), B=@n5))
  n6 = TextureCoordinate()
  Add(A=@n6, B=Constant2Vector() {r=-0.5})
  n7 = Time()
  n8 = Multiply(A=@n7, B=ScalarParameter() {defaultValue=5, parameterName=Time2})
  n9 = ScalarParameter() {defaultValue=0.7, parameterName=floor}
  n10 = Multiply(A=@n7, B=ScalarParameter() {defaultValue=2.16, parameterName=Time1})
  n11 = ScalarParameter() {defaultValue=1.1, parameterName=floor2}
  n12 = ScalarParameter() {defaultValue=4.1, parameterName=Time3}
  n13 = Multiply(
   A=Multiply(
    A=MaterialFunctionCall!(Sine Phase=@n8, Value 1=@n9, Value 2) {materialFunction=/Engine/Functions/Engine_MaterialFunctions02/Utility/Sine_Remapped}.Remapped Sine
    B=MaterialFunctionCall(Sine Phase=@n10, Value 1=@n11) {materialFunction=&Sine_Remapped}.Remapped Sine
   )
   B=MaterialFunctionCall(Sine Phase=Multiply(A=@n7, B=@n12), Value 1=ScalarParameter() {defaultValue=0.75, parameterName=floor3}) {materialFunction=&Sine_Remapped}.Remapped Sine
  )
  ComponentMask(None=@n13) {g=true, r=true}
  FeatureLevelSwitch()
  n14 = ScalarParameter() {defaultValue=1, parameterName=Fadein}
  n15 = Constant()
  If(A=@n14, B=Constant(), "A > B"=Constant() {r=1}, "A == B"=@n15, "A < B"=@n15)
  LinearInterpolate() {constA=1, constB=10}
  n16 = ScalarParameter() {defaultValue=-0.5, parameterName=xadd}
  n17 = ScalarParameter() {defaultValue=-2.7572, parameterName=yadd}
  n18 = Multiply(
   A=Add(A=@n6, B=AppendVector(A=@n16, B=@n17))
   B=AppendVector(A=ScalarParameter() {defaultValue=-1, parameterName=x}, B=ScalarParameter() {defaultValue=0.3448, parameterName=y})
  )
  n19 = Time()
  n20 = ScalarParameter() {defaultValue=1.1419, parameterName=SpeedRainbow}
  n21 = TextureSample(UVs=Panner(
    Coordinate=MaterialFunctionCall(Vector or UVs=@n18) {materialFunction=&VectorToRadialValue}.Radial Coordinates
    Time=Multiply(A=@n19, B=@n20)
   ) {speedX=1.667}) {texture=/Game/Global/Textures/Default/Masks_and_Gradients/T_Grad_Rainbow}
  n22 = MaterialFunctionCall(LDR Input=Clamp(None=@n21.R), HDR Tint=@n21.R) {materialFunction=&DeriveHDRfromLDR}
  n23 = Multiply(A=@n13, B=VectorParameter() {parameterName=ColorClock, defaultValue=<rgba 0.138 0.4213 0.5 1>}.RGB)
  n24 = TextureCoordinate()
  n25 = Constant2Vector() {g=0.5, r=0.5}
  n26 = TextureSample(UVs=Divide(A=Distance(A=@n24, B=@n25), B=ScalarParameter() {defaultValue=2.7181, parameterName=Chromatic})) {texture=/Game/Effects/Fort_Effects/Textures/Tile/T_Inky_Smoke_Tile_Chromatic}
  n27 = Add(A=@n23, B=@n26.RGB)
  n28 = AppendVector(
   A=ScalarParameter() {defaultValue=1, group=Numbers, parameterName=u_NumberPaddingScale}
   B=ScalarParameter() {defaultValue=0.4667, group=Numbers, parameterName=v_NumberPaddingScale}
  )
  n29 = MaterialFunctionCall!(UVs, UVPadding=@n28, "NumberofDigits(int)"=Constant() {r=8}) {materialFunction=/Game/Packages/Fortress_SharedMaterials/Base_Material_Functions/BaseCharacterShaderFunctions/MF_CountdownTimer}
  n30 = Multiply(A=TextureCoordinate(), B=ScalarParameter() {defaultValue=1.1383, parameterName=BlocksTiling})
  n31 = TextureSample(UVs=@n30) {samplerType=SAMPLERTYPE_Masks, texture=/Game/Athena/Playlists/Barrier/Textures/T_BarrierGrid_M}
  n32 = ComponentMask(None=@n31.RGB) {r=true}
  n33 = Ceil(None=Abs(None=@n30))
  n34 = ComponentMask(None=@n33) {r=true}
  n35 = ComponentMask(None=@n33) {g=true}
  n36 = Multiply(A=Add(A=Time() {period=1}, B=Add(A=@n34, B=@n35)), B=ScalarParameter() {defaultValue=1.3215, parameterName=Sine Speed Grid})
  n37 = Sine(None=Frac(None=Add(A=@n32, B=@n36)))
  n38 = ComponentMask(None=@n31.RGB) {g=true}
  n39 = Multiply(A=Power(Base=Multiply(A=@n37, B=@n38), Exp=ScalarParameter() {defaultValue=5, parameterName=Grid Power}), B=@n29.0-1 Mask)
  n40 = Multiply(A=@n39, B=@n22.RemappedRange)
  n41 = StaticSwitchParameter(
   True=Constant() {r=1}
   False=CollectionParameter() {collection=/Game/Athena/Events/Generic/CountdownTimer/Materials/MPC_Countdown, parameterName=ShowCountdownPhase3}
  ) {parameterName=Debug2}
  LinearInterpolate(
   A=Multiply(A=StaticSwitchParameter(True=@n22.RemappedRange, False=@n27) {parameterName=UseRainbowEffect}, B=@n29.NumbersGreyscale)
   B=Multiply(A=StaticSwitchParameter(True=@n40, False=@n39) {parameterName=UseRainbowEffect}, B=@n27)
   Alpha=LinearInterpolate(A=ScalarParameter() {defaultValue=0.02, parameterName=ShowMarshMello}, B=Constant() {r=1}, Alpha=@n41)
  )
  n42 = ObjectPositionWS()
  n43 = WorldPosition()
  n44 = CollectionParameter() {collection=&MPC_Countdown, parameterName=ShowCountdown}
  n45 = Multiply(
   A=@n44
   B=CollectionParameter() {collection=&MPC_Countdown, parameterName=ShowCountdownPhase2}
  )
  LinearInterpolate(
   A=Add(A=Subtract(A=@n42, B=@n43.XYZ), B=Constant3Vector() {constant=<rgba 0 0 -5156156 0>})
   B=StaticSwitchParameter(True=@n3.World position offset, False=Constant3Vector()) {parameterName=IsCameraFacing}
   Alpha=StaticSwitchParameter(True=Constant() {r=1}, False=StaticSwitchParameter(True=@n45, False=@n44) {parameterName=isPhase2}) {parameterName=Debug}
  )
  n46 = Frac(None=Divide(A=ScreenPosition().PixelPosition) {constB=5})
  n47 = Constant() {r=0.5}
  n48 = ScalarParameter() {defaultValue=1.0191, parameterName=Param_1}
  n49 = Constant()
  MaterialFunctionCall!(
   Dynamic Param=ScalarParameter() {defaultValue=0.2, parameterName=Erode}
   Alpha=Multiply(A=SphereMask(A=@n46, B=@n47, Radius=@n48, Hardness=@n49), B=LinearInterpolate(A=@n29.NumbersGreyscale, B=@n39, Alpha=@n41))
  ) {materialFunction=/Game/Effects/Fort_Effects/Materials/_Functions/Erode_Alpha}
  Multiply(A=@n23, B=@n26.RGB)
  Multiply(A=@n28, B=LinearInterpolate(A=Constant() {r=1}, B=Constant() {r=1.3}, Alpha=ScalarParameter() {parameterName=Pulsing}))
  n50 = Divide(A=Distance(A=ActorPositionWS(), B=CameraPositionWS()), B=ScalarParameter() {defaultValue=2117.0803, parameterName=Bias})
  n51 = ScalarParameter() {defaultValue=0.3048, parameterName=ScaleMin}
  n52 = ScalarParameter() {defaultValue=1.5, parameterName=ScaleMax}
  Multiply(
   A=Clamp(None=Multiply(A=@n1, B=@n50), Min=Multiply(A=@n1, B=@n51), Max=Multiply(A=@n1, B=@n52))
   B=Multiply(A=ScalarParameter() {defaultValue=1, parameterName=Scale_Master}, B=@n2.Scale Z)
  )
  Multiply(A=@n5, B=@n2.Scale Z)
  Multiply()
  n53 = Subtract(A=SceneDepth().R, B=ScalarParameter() {defaultValue=47881.7031, parameterName=ScenDepthFadeDistance})
  n54 = ScalarParameter() {defaultValue=2861.6772, parameterName=ScenDepthFade}
  Multiply(A=Multiply(A=@n29.NumbersGreyscale, B=@n14), B=Clamp(None=Divide(A=@n53, B=@n54)))
  PixelDepth()
  ScalarParameter() {defaultValue=1.5, parameterName=BlocksTiling2}
 ]
}
