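# m_WhiteOut is a prototype translucent, unlit material, apparently for a depth-based white-out effect.
# MF_DistanceBehindTranslucency_01 measures normalized distance behind the surface using the Depth parameter, and a Power node shapes that falloff with Param_1.
# A pale blue Color, Param_2, DepthFade and WorldPositionBehindTranslucency are present but unwired, and no links to material outputs appear.
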Material /Figment_S08/Athena/Prototype/Blueprints/White/m_WhiteOut
graph {
 properties {blendMode=BLEND_Translucent, pixelDepthOffsetMode=PDOM_Legacy, shadingModel=MSM_Unlit, usedShadingModels=Unlit}
 nodes [
  DepthFade()
  MaterialFunctionCall() {materialFunction=&WorldPositionBehindTranslucency}
  Power(
   Base=MaterialFunctionCall!(Distance Bias=Constant(), Distance Divisor=ScalarParameter() {defaultValue=800, parameterName=Depth}, Scene Depth, Pixel Depth) {materialFunction=/Game/Effects/Fort_Effects/Materials/_Functions/MF_DistanceBehindTranslucency_01}.Normalized
   Exp=ScalarParameter() {defaultValue=3, parameterName=Param_1}
  )
  ScalarParameter()
  ScalarParameter() {defaultValue=5, parameterName=Param_2}
  VectorParameter() {parameterName=Color, defaultValue=<rgba 0.6615 0.9058 1 1>}
 ]
}
